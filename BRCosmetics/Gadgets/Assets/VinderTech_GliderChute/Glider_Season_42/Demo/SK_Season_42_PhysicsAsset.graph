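# Physics asset for the Season 42 demo glider skeletal mesh in the Battle Royale cosmetics glider chute gadget set.
# It holds a single capsule body on the Root bone, offset upward from the bone origin.
# It has no constraints, so the glider is one rigid collision volume with no jointed or per-bone simulation.
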PhysicsAsset /BRCosmetics/Gadgets/Assets/VinderTech_GliderChute/Glider_Season_42/Demo/SK_Season_42_PhysicsAsset
bodies {Root [{shapeType=Capsule, length=66.4625, radius=61.6588, center=<0.0962 -0.1721 49.1851>, rotation=<rot 0 3.0266 -179.8567>}]}
constraints []
